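# Silver_Blazer_Close_Cue is a sound cue in the Bubble Shield item sounds, apparently played when the shield closes.
# A single wave player feeds the Silver_Blazer_Close_01 wave straight to the output at reduced volume, using the bubble shield open/close attenuation and the Fort_AGV sound class.
# A concurrency override allows one instance per owner and blocks new plays while it is active.
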SoundCue /Game/Athena/Sounds/Items/Bubble_Shield/Silver_Blazer_Close_Cue
root=WavePlayer_0
nodes [{id=WavePlayer_0, properties {soundWaveAssetPtr=&/Silver_Blazer_Close_01}}]
properties {
 attenuationSettings=/Game/Sounds/Attenuation/Fort_Weapon_Bubble_Shield_Open_Close_Att
 bOverrideConcurrency=true
 maxDistance=10800
 soundClassObject=/Game/Sounds/FortSoundClassesAndModes/Fort_AGV
 volumeMultiplier=0.5
 concurrencyOverrides {bLimitToOwner=true, maxCount=1, resolutionRule=PreventNew}
}
