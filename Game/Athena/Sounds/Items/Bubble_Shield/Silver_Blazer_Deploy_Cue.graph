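# Sound cue for the deploy sound of the Silver Blazer bubble shield item.
# A single wave player feeds the deploy wave straight to the output, with a reduced volume multiplier.
# It uses bubble shield open/close attenuation and the Fort_AGV sound class, and a concurrency override caps simultaneous instances, stopping the oldest.
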SoundCue /Game/Athena/Sounds/Items/Bubble_Shield/Silver_Blazer_Deploy_Cue
root=WavePlayer_5
nodes [{id=WavePlayer_5, properties {soundWaveAssetPtr=&/Silver_Blazer_Deploy_01}}]
properties {
 attenuationSettings=/Game/Sounds/Attenuation/Fort_Weapon_Bubble_Shield_Open_Close_Att
 bOverrideConcurrency=true
 maxDistance=10800
 soundClassObject=/Game/Sounds/FortSoundClassesAndModes/Fort_AGV
 volumeMultiplier=0.6
 concurrencyOverrides {maxCount=3, resolutionRule=StopOldest}
}
