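# Sound cue for entering the Silver Blazer bubble shield in Fortnite Athena.
# A single wave player feeds a modulator that randomizes pitch slightly within a lowered range, played at reduced volume.
# Uses the Silver Blazer enter/exit attenuation and a third-person ranged-weapon fire sound class, with a concurrency override that stops the oldest of a few simultaneous instances.
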SoundCue /Game/Athena/Sounds/Items/Bubble_Shield/Silver_Blazer_Enter_Cue
root=Modulator_0
nodes [
 {id=Modulator_0, children [WavePlayer_1], properties {pitchMax=0.85, pitchMin=0.8}}
 {id=WavePlayer_1, properties {soundWaveAssetPtr=&/Silver_Blazer_Enter_01}}
]
properties {
 attenuationSettings=&/ATT_SilverBlazer_EnterExit
 bOverrideConcurrency=true
 maxDistance=7500
 soundClassObject=&Fort_WeaponRanged_Fire_3P_PSM
 volumeMultiplier=0.4
 concurrencyOverrides {maxCount=3, resolutionRule=StopOldest}
}
