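# Sound cue that plays the equip sound for the Silver Blazer item in the Bubble Shield item sound folder.
# A single wave player feeds the root directly, with no randomization or modulation.
# It borrows weapon reload attenuation, uses the loot collect sound class and overrides concurrency to cap simultaneous instances.
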SoundCue /Game/Athena/Sounds/Items/Bubble_Shield/Silver_Blazer_Equip_Cue
root=WavePlayer_1
nodes [{id=WavePlayer_1, properties {soundWaveAssetPtr=&/Silver_Blazer_Equip_01}}]
properties {
 attenuationSettings=&Fort_Weapon_Reload_Att
 bOverrideConcurrency=true
 maxDistance=2250
 soundClassObject=&Fort_Loot_Collect
 concurrencyOverrides {maxCount=3}
}
