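# Sound cue for impacts on the Bubble Shield item (Silver Blazer) in Athena.
# A Random node picks one of six equally weighted impact waves, and a Modulator applies a slight random volume boost.
# It plays through the Fort_AGV sound class with bullet impact attenuation at a slightly reduced overall volume.
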SoundCue /Game/Athena/Sounds/Items/Bubble_Shield/Silver_Blazer_Impact_Cue
root=Modulator_0
nodes [
 {id=Modulator_0, children [Random_0], properties {volumeMax=1.25, volumeMin=1.2}}
 {id=Random_0, children [WavePlayer_0,WavePlayer_1,WavePlayer_2,WavePlayer_3,WavePlayer_4,WavePlayer_5], properties {weights [1,1,1,1,1,1]}}
 {id=WavePlayer_0, properties {soundWaveAssetPtr=&/Silver_Blazer_Impact_01}}
 {id=WavePlayer_1, properties {soundWaveAssetPtr=&/Silver_Blazer_Impact_02}}
 {id=WavePlayer_2, properties {soundWaveAssetPtr=&/Silver_Blazer_Impact_03}}
 {id=WavePlayer_3, properties {soundWaveAssetPtr=&/Silver_Blazer_Impact_04}}
 {id=WavePlayer_4, properties {soundWaveAssetPtr=&/Silver_Blazer_Impact_05}}
 {id=WavePlayer_5, properties {soundWaveAssetPtr=&/Silver_Blazer_Impact_06}}
]
properties {
 attenuationSettings=/Game/Sounds/Attenuation/Fort_Bullet_Impact_Athena_Att
 maxDistance=4500
 soundClassObject=/Game/Sounds/FortSoundClassesAndModes/Fort_AGV
 volumeMultiplier=0.8
}
