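# Sound cue for the Bubble Shield item's looping ambience, likely its exterior variant going by the "Ext" name.
# A Mixer layers a looping shimmer wave, boosted in volume, over a second looping Silver Blazer wave.
# It uses the bubble shield ambient attenuation and the Fort_AGV sound class at reduced overall volume.
# A concurrency override caps simultaneous instances, with voice-steal release and volume scaling to tame stacked shields.
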SoundCue /Game/Athena/Sounds/Items/Bubble_Shield/Silver_Blazer_Loop_Ext_Cue
root=Mixer_0
nodes [
 {id=Mixer_0, children [WavePlayer_1,WavePlayer_0], properties {inputVolume [1.8189,1]}}
 {id=WavePlayer_0, properties {bLooping=true, soundWaveAssetPtr=&/Silver_Blazer_Loop_02}}
 {id=WavePlayer_1, properties {bLooping=true, soundWaveAssetPtr=&/Silver_Blazer_Shimmer_Loop_01}}
]
properties {
 attenuationSettings=/Game/Sounds/Attenuation/Fort_Weapon_Bubble_Shield_Amb_Att
 bOverrideConcurrency=true
 maxDistance=3600
 soundClassObject=/Game/Sounds/FortSoundClassesAndModes/Fort_AGV
 volumeMultiplier=0.5
 concurrencyOverrides {maxCount=3, voiceStealReleaseTime=0.5, volumeScale=0.75, volumeScaleAttackTime=0.25}
}
