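# Sound cue for the warning pulse of the Silver Blazer item in the Athena Bubble Shield sound set.
# A single wave player node feeds the root directly, playing one fixed warning pulse wave at reduced volume with no randomization or modulation.
# The cue overrides its referenced bullet-impact attenuation with inline settings: logarithmic falloff, listener focus and a non-spatialized radius near the source.
# It routes through the Fort_AGV sound class.
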SoundCue /Game/Athena/Sounds/Items/Bubble_Shield/Silver_Blazer_Warning_Pulse_Cue
root=WavePlayer_1
nodes [{id=WavePlayer_1, properties {soundWaveAssetPtr=&/Silver_Blazer_Warning_Pulse_01}}]
properties {
 attenuationSettings=/Game/Sounds/Attenuation/Fort_Bullet_Impact_Athena_Att
 bOverrideAttenuation=true
 maxDistance=4000
 soundClassObject=/Game/Sounds/FortSoundClassesAndModes/Fort_AGV
 volumeMultiplier=0.7
 attenuationOverrides {
  bEnableListenerFocus=true
  distanceAlgorithm=Logarithmic
  falloffDistance=2500
  nonFocusVolumeAttenuation=0.8
  nonSpatializedRadiusStart=500
  attenuationShapeExtents {x=1500}
 }
}
